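# Sound cue for the pump shotgun's empty dry-fire sound in the WeaponsUnvaulted content.
# A single wave player plays one stock Fortnite dry-fire wave at reduced volume, with no randomization or mixing.
# The cue routes to the weapon foley sound class and overrides concurrency to prevent new instances once the limit is reached.
# It carries a falloff-distance value in its attenuation overrides, but the attenuation override flag is not set.
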SoundCue /WeaponsUnvaulted/Assets/Audio/PumpShotgun/SC_ShotgunPump_Empty
root=WavePlayer_0
nodes [{id=WavePlayer_0, properties {soundWaveAssetPtr=/Game/Sounds/Fort_Weapons/Shotgun/CombatNew/Shotgun_CombatNew_DryFire_01}}]
properties {
 bOverrideConcurrency=true
 maxDistance=2097152
 soundClassObject=&Fort_WeaponFoley
 volumeMultiplier=0.501
 attenuationOverrides {falloffDistance=1600}
 concurrencyOverrides {resolutionRule=PreventNew}
}
