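# Sound cue for the cartridge impact sound heard during pump shotgun reloads.
# A random node picks one of six equally weighted impact variations and feeds it through a modulator at reduced cue volume.
# It plays in the weapon foley sound class with reload attenuation, and overrides concurrency to limit overlapping copies per owner, stopping the oldest.
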SoundCue /WeaponsUnvaulted/Assets/Audio/PumpShotgun/SC_ShotgunPump_Reload_CartridgeImpact
root=Modulator_14
nodes [
 {id=Modulator_14, children [Random_1]}
 {id=Random_1, children [WavePlayer_0,WavePlayer_5,WavePlayer_6,WavePlayer_7,WavePlayer_8,WavePlayer_9], properties {weights [1,1,1,1,1,1]}}
 {id=WavePlayer_0, properties {soundWaveAssetPtr=/WeaponsUnvaulted/Assets/Audio/PumpShotgun/SoundWaves/ShotgunPump_Reload_CartridgeImpact_01}}
 {id=WavePlayer_5, properties {soundWaveAssetPtr=/WeaponsUnvaulted/Assets/Audio/PumpShotgun/SoundWaves/ShotgunPump_Reload_CartridgeImpact_02}}
 {id=WavePlayer_6, properties {soundWaveAssetPtr=/WeaponsUnvaulted/Assets/Audio/PumpShotgun/SoundWaves/ShotgunPump_Reload_CartridgeImpact_03}}
 {id=WavePlayer_7, properties {soundWaveAssetPtr=/WeaponsUnvaulted/Assets/Audio/PumpShotgun/SoundWaves/ShotgunPump_Reload_CartridgeImpact_04}}
 {id=WavePlayer_8, properties {soundWaveAssetPtr=/WeaponsUnvaulted/Assets/Audio/PumpShotgun/SoundWaves/ShotgunPump_Reload_CartridgeImpact_05}}
 {id=WavePlayer_9, properties {soundWaveAssetPtr=/WeaponsUnvaulted/Assets/Audio/PumpShotgun/SoundWaves/ShotgunPump_Reload_CartridgeImpact_06}}
]
properties {
 attenuationSettings=&Fort_Weapon_Reload_Att
 bOverrideConcurrency=true
 maxDistance=2250
 soundClassObject=&Fort_WeaponFoley
 volumeMultiplier=0.4
 attenuationOverrides {falloffDistance=1600}
 concurrencyOverrides {bLimitToOwner=true, maxCount=2, resolutionRule=StopOldest}
}
