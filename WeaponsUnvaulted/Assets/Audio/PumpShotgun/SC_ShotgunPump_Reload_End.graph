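# Sound cue for the end of the pump shotgun's reload, played as weapon foley.
# A random node picks one of four equally weighted wave variations and passes it through a modulator for playback variation.
# It uses the Fortnite weapon reload attenuation at reduced volume, with a concurrency override limiting it to two instances per owner, stopping the oldest.
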SoundCue /WeaponsUnvaulted/Assets/Audio/PumpShotgun/SC_ShotgunPump_Reload_End
root=Modulator_14
nodes [
 {id=Modulator_14, children [Random_0]}
 {id=Random_0, children [WavePlayer_1,WavePlayer_2,WavePlayer_3,WavePlayer_4], properties {weights [1,1,1,1]}}
 {id=WavePlayer_1, properties {soundWaveAssetPtr=/WeaponsUnvaulted/Assets/Audio/PumpShotgun/SoundWaves/ShotgunPump_Reload_End_04}}
 {id=WavePlayer_2, properties {soundWaveAssetPtr=/WeaponsUnvaulted/Assets/Audio/PumpShotgun/SoundWaves/ShotgunPump_Reload_End_03}}
 {id=WavePlayer_3, properties {soundWaveAssetPtr=/WeaponsUnvaulted/Assets/Audio/PumpShotgun/SoundWaves/ShotgunPump_Reload_End_02}}
 {id=WavePlayer_4, properties {soundWaveAssetPtr=/WeaponsUnvaulted/Assets/Audio/PumpShotgun/SoundWaves/ShotgunPump_Reload_End_01}}
]
properties {
 attenuationSettings=&Fort_Weapon_Reload_Att
 bOverrideConcurrency=true
 maxDistance=2250
 soundClassObject=&Fort_WeaponFoley
 volumeMultiplier=0.3
 attenuationOverrides {falloffDistance=1600}
 concurrencyOverrides {bLimitToOwner=true, maxCount=2, resolutionRule=StopOldest}
}
