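# Sound cue for the pump shotgun's shell-insert reload sound.
# A Random node picks one of three insert waves with equal weight and feeds a Modulator that holds volume fixed.
# It plays through the weapon foley sound class with weapon reload attenuation, and a concurrency override prevents new instances once the limit is reached.
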SoundCue /WeaponsUnvaulted/Assets/Audio/PumpShotgun/SC_ShotgunPump_Reload_Insert
root=Modulator_13
nodes [
 {id=Modulator_13, children [Random_1], properties {volumeMax=1, volumeMin=1}}
 {id=Random_1, children [WavePlayer_0,WavePlayer_1,WavePlayer_2], properties {weights [1,1,1]}}
 {id=WavePlayer_0, properties {soundWaveAssetPtr=/WeaponsUnvaulted/Assets/Audio/PumpShotgun/SoundWaves/ShotgunPump_Reload_Insert_01}}
 {id=WavePlayer_1, properties {soundWaveAssetPtr=/WeaponsUnvaulted/Assets/Audio/PumpShotgun/SoundWaves/ShotgunPump_Reload_Insert_02}}
 {id=WavePlayer_2, properties {soundWaveAssetPtr=/WeaponsUnvaulted/Assets/Audio/PumpShotgun/SoundWaves/ShotgunPump_Reload_Insert_03}}
]
properties {
 attenuationSettings=&Fort_Weapon_Reload_Att
 bOverrideConcurrency=true
 maxDistance=2250
 soundClassObject=&Fort_WeaponFoley
 volumeMultiplier=1
 attenuationOverrides {falloffDistance=1600}
 concurrencyOverrides {resolutionRule=PreventNew}
}
